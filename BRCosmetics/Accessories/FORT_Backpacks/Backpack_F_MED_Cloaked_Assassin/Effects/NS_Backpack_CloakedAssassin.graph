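# Niagara effect for the Cloaked Assassin backpack cosmetic, adding purple lightning, flying rune sprites and glow lights that intensify with the User.NumberKills count.
# Flickering lightning sprites and lights follow sampled points on the user skeletal mesh, with spawn rate, brightness and light timing ramped by kills.
# BigLightningEndPicker picks random socket pairs; BigLightningRibbon reads them through a particle attribute reader and draws jittered, tapered arcs between them.
# Runes burst outward from a small sphere, decelerating as they spin.
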NiagaraSystem /BRCosmetics/Accessories/FORT_Backpacks/Backpack_F_MED_Cloaked_Assassin/Effects/NS_Backpack_CloakedAssassin
dependencies {
 usedDataInterfaces [&NiagaraDataInterfaceCurve, &NiagaraDataInterfaceColorCurve, &NiagaraDataInterfaceParticleRead]
 usedDynamicInputs []
 usedModules [&SystemState]
}
emitters {
 BigLightningEndPicker {
  emitterUpdateScript [
   /SetVariables(Emitter.SetLoopDelay:NiagaraFloat=&UniformRangedFloat(Minimum=0.1, Maximum=0.9)) {isSetParameters=true}
   /SetVariables(Emitter.SetLoopDuration:NiagaraFloat=&Subtract_Float(A=1, B=$Emitter.SetLoopDelay)) {isSetParameters=true}
   &EmitterState(
    Life Cycle Mode=::NewEnumerator1
    Loop Delay=$Emitter.SetLoopDelay
    Loop Duration=$Emitter.SetLoopDuration
    Recalculate Duration Each Loop=true
    UseLoopDelay=true
   )
   &SpawnRate(
    Spawn Probability=0.3333
    SpawnRate=&ClampFloat(
     Float=&Add_Float(A=-2.3333, B=&Multiply_Float(A=1.3333, B=$User.NumberKills))
     Max=3
    )
    Use Spawn Probability=true
   )
  ]
  particleSpawnScript [
   /SetVariables(
    Particles.BottomSockets:NiagaraInt32=&UniformRangedInt(Minimum=2, Maximum=4)
    Particles.LeftSockets:NiagaraInt32=&UniformRangedInt(Minimum=3, Maximum=5)
    Particles.LightningDirection:NiagaraInt32=&UniformRangedInt(Minimum=0, Maximum=1)
    Particles.RightSockets:NiagaraInt32=&UniformRangedInt(Minimum=0, Maximum=2)
    Particles.TopSockets:NiagaraInt32=&ModuloInt(A=&UniformRangedInt(Minimum=5, Maximum=7), B=6)
   ) {isSetParameters=true}
   /Niagara/Modules/Spawn/Initialization/InitializeParticle(
    Color=&UniformRangedLinearColor(Minimum=<rgba 0.2 0 1 1>, Maximum=<rgba 2 0 3 1>, Color Channel Mode=::NewEnumerator2)
    Lifetime=&UniformRangedFloat(Minimum=0.2, Maximum=0.3)
    Module.Write Color=true
    Module.Write Lifetime=true
    Module.Write Mass=true
    Module.Write Material Random=true
    Module.Write Position=false
    Module.Write Scale=false
    Module.Write SpriteRotation=true
    Module.Write SpriteSize=true
    Position=&SimulationPosition()
    Sprite Rotation=90
    Sprite Size=&Vector2DFromFloat(Value=5)
   )
   &SkeletalMeshLocation(
    Bone / Socket Index=&MakeIntFromFloat(Float=&Lerp_Float(
      A=&MakeFloatFromInt(Int=$Particles.TopSockets)
      B=&MakeFloatFromInt(Int=$Particles.LeftSockets)
      Alpha=&MakeFloatFromInt(Int=$Particles.LightningDirection)
     ))
    Inherit Velocity=false
    MaskSpawnGroup=false
    Mesh Sampling Type=::NewEnumerator5
    Skeletal Mesh=$User.SkeletalMesh
    Socket Sampling Mode=::NewEnumerator5
   ) {module=SkeletalMeshLocation001}
   /SetVariables(Particles.HomePosition:Vector3f=$Output.SkeletalMeshLocation001.SampledPosition) {isSetParameters=true}
   &SkeletalMeshLocation(
    Apply To Particle Position=false
    Bone / Socket Index=&MakeIntFromFloat(Float=&Lerp_Float(
      A=&MakeFloatFromInt(Int=$Particles.BottomSockets)
      B=&MakeFloatFromInt(Int=$Particles.RightSockets)
      Alpha=&MakeFloatFromInt(Int=$Particles.LightningDirection)
     ))
    Inherit Velocity=false
    MaskSpawnGroup=false
    Mesh Sampling Type=::NewEnumerator5
    Skeletal Mesh=$User.SkeletalMesh
    Socket Sampling Mode=::NewEnumerator5
   )
   /SetVariables(Particles.FriendPosition:Vector3f=$Output.SkeletalMeshLocation.SampledPosition) {isSetParameters=true}
  ]
  particleUpdateScript [
   &SolveForcesAndVelocity(Clamp Velocity=false, Limit Acceleration=false)
   &ParticleState()
  ]
  rendererClasses [&NiagaraSpriteRendererProperties]
  renderers [
   {rendererClass=&NiagaraSpriteRendererProperties}
   {rendererClass=&NiagaraSpriteRendererProperties, rendererIndex=1}
  ]
 }
 BigLightningRibbon {
  emitterSpawnScript [/Game/Effects/Niagara/Modules/Attributes/FN_Source_v2(AttributeReader=DataInterface() {propertyValues {EmitterBinding {bindingMode=Other, emitterName=BigLightningEndPicker}}})]
  emitterUpdateScript [
   &EmitterState()
   /SetVariables(Emitter.ParticlesPerRibbon:NiagaraFloat=&UniformRangedFloat(Minimum=15, Maximum=30)) {isSetParameters=true}
   &SpawnRate(
    SpawnRate=/Game/Effects/Niagara/DynamicInputs/Attributes/FN_GetEmitterParticleCount(AttributeReader:NiagaraDataInterfaceParticleRead=$Emitter.AttributeReader, Multiplier:NiagaraFloat=$Emitter.ParticlesPerRibbon)
    Use Spawn Probability=false
   )
  ]
  particleSpawnScript [
   &self:GetInfoFromParent() {module=GetInfoFromParent}
   /SetVariables(Particles.AgeByParent:NiagaraFloat=&FloatFromCurve(FloatCurve=DataInterface() {propertyValues {Curve {keys [<at 0.2 1>,<at 0.8 0>]}}}, CurveIndex=$Particles.AgeImported)) {isSetParameters=true}
   /SetVariables(Particles.DistanceBetweenParents:NiagaraFloat=&ClampFloat(Float=&Add_Float(
      A=-0.5455
      B=&Multiply_Float(
       A=0.0909
       B=/Niagara/DynamicInputs/TypeConversions/VectorLengthSafe(Vector:Vector3f=&Subtract_Vector(A=$Particles.ParentPosition, B=$Particles.ParentFriendPosition))
      )
     ))) {isSetParameters=true}
   &InitializeRibbon(
    Color=<rgba 2.4952 0 6 1>
    Lifetime=0.2
    Module.Write Color=true
    Module.Write Lifetime=true
    Module.Write Mass=false
    Module.Write Material Random=true
    Module.Write Position=false
    Module.Write RibbonTwist=false
    Module.Write RibbonWidth=true
    Ribbon Width=&Multiply_Float(
     A=/Niagara/DynamicInputs/Math/LerpMultipleFloats(
      Float Count:ENiagara_MultipleLerpCount=::NewEnumerator0
      1:NiagaraFloat=2
      2:NiagaraFloat=&UniformRangedFloat(Minimum=10, Maximum=30)
      3:NiagaraFloat=2
      4:NiagaraFloat=?
      5:NiagaraFloat=?
      Alpha:NiagaraFloat=$Particles.AgeByParent
     )
     B=&Lerp_Float(A=0.75, B=1, Alpha=$Particles.DistanceBetweenParents)
    )
   )
   &InitializeParticle(
    Color=&ColorFromCurve(
     ColorCurve=DataInterface() {propertyValues {
       AlphaCurve {keys [<at 0 1>,<at 1 1>]}
       BlueCurve {keys [<at 0.0171 0>,<at 0.4894 0.0261>,<at 1.0012 1>]}
       GreenCurve {keys [<at 0.0171 0>,<at 0.4894 1>,<at 1.0012 0>]}
       RedCurve {keys [<at 0.0171 1>,<at 0.4894 0>,<at 1.0012 0>]}
      }}
     CurveIndex=$Particles.AgeByParent
    )
    Lifetime=0.7
    Module.Write Color=false
    Module.Write Lifetime=false
    Module.Write Mass=false
    Module.Write Material Random=true
    Module.Write Position=false
    Module.Write Scale=false
    Module.Write SpriteRotation=false
    Module.Write SpriteSize=true
    Sprite Size=&Vector2DFromFloat(Value=&Multiply_Float(A=1, B=1))
   )
   /SetVariables(
    Particles.OffsetPerRibbon:Vector3f=&UniformRangedVector(Minimum=<-5 -7 -5>, Maximum=<5 1 5>, Randomness Mode=::NewEnumerator1, Random Seed=$Particles.ParentIDInt, Fixed Random Seed=true)
    Particles.OffsetRandom:Vector3f=&UniformRangedVector(Minimum=<-1 -1 -1>, Maximum=<1 1 1>)
    Particles.ParentsLerp:Vector3f=&Lerp_Vector(
     A=$Particles.ParentPosition
     B=$Particles.ParentFriendPosition
     Alpha=&ClampFloat(Float=&Add_Float(A=-0.1, B=&Multiply_Float(A=1.2, B=$Particles.AgeByParent)))
    )
   ) {isSetParameters=true}
   /SetVariables(Particles.Position=&Add_Vector(
     A=$Particles.ParentsLerp
     B=&Multiply_VectorByFloat(
      Vector=&Add_Vector(A=$Particles.OffsetPerRibbon, B=$Particles.OffsetRandom)
      Float=&FloatFromCurve(
       FloatCurve=DataInterface() {propertyValues {Curve {keys [
           {arriveTangent=2.9233, interpMode=RCIM_Cubic, leaveTangent=2.9233, tangentMode=RCTM_User, time=0.1, value=0}
           {interpMode=RCIM_Cubic, tangentMode=RCTM_User, time=0.4951, value=1}
           {arriveTangent=-2.7178, interpMode=RCIM_Cubic, leaveTangent=-2.7178, tangentMode=RCTM_User, time=0.9, value=0}
          ]}}}
       CurveIndex=$Particles.AgeByParent
      )
     )
    )) {isSetParameters=true}
   &AddVelocity(Velocity=&Multiply_VectorByFloat(
     Vector=&Multiply_Vector(A=$Particles.OffsetRandom, B=<-40 5 -40>)
     Float=&FloatFromCurve(
      FloatCurve=DataInterface() {propertyValues {Curve {keys [
          {arriveTangent=0.0416, interpMode=RCIM_Cubic, leaveTangent=0.0416, tangentMode=RCTM_User, time=0, value=0.2}
          {interpMode=RCIM_Cubic, tangentMode=RCTM_User, time=0.5, value=1}
          {arriveTangent=-0.0664, interpMode=RCIM_Cubic, leaveTangent=-0.0664, tangentMode=RCTM_User, time=1, value=0.2}
         ]}}}
      CurveIndex=$Particles.AgeByParent
      Scale Curve=$Particles.DistanceBetweenParents
     )
    ))
  ]
  particleUpdateScript [
   &ParticleState()
   &Drag(Drag=5, Use Linear Drag=true, Use Rotational Drag=false)
   &SolveForcesAndVelocity(Clamp Velocity=false, Limit Acceleration=false)
   &ScaleRibbonWidth(Ribbon Width Scale=&FloatFromCurve(FloatCurve=DataInterface() {propertyValues {Curve {keys [
         {arriveTangent=7.8756, interpMode=RCIM_Cubic, leaveTangent=7.8756, tangentMode=RCTM_User, time=0, value=0}
         {interpMode=RCIM_Cubic, tangentMode=RCTM_User, time=0.2369, value=1}
         {arriveTangent=-2.4215, interpMode=RCIM_Cubic, leaveTangent=-2.4215, tangentMode=RCTM_User, time=1, value=0}
        ]}}}))
   &ScaleColor(
    Scale RGB=&VectorFromFloat(Value=&LerpMultipleFloats(
      Float Count:ENiagara_MultipleLerpCount=::NewEnumerator2
      1:NiagaraFloat=1.5
      2:NiagaraFloat=0.5
      3:NiagaraFloat=0.5
      4:NiagaraFloat=0.5
      5:NiagaraFloat=0.1
      Alpha:NiagaraFloat=$Particles.NormalizedAge
     ))
    ScaleA=false
    ScaleRGB=true
   )
  ]
  rendererClasses [&NiagaraRibbonRendererProperties, &NiagaraSpriteRendererProperties]
  renderers [
   {rendererClass=&NiagaraRibbonRendererProperties}
   {rendererClass=&NiagaraSpriteRendererProperties, rendererIndex=1}
  ]
 }
 Lightning {
  emitterUpdateScript [
   &EmitterState()
   /SetVariables(
    Emitter.SpawnProbabilityMath:NiagaraFloat=&ClampFloat(Float=&Add_Float(A=-1.85, B=&Multiply_Float(A=0.95, B=$User.NumberKills)))
    Emitter.SpawnRateMath:NiagaraFloat=&ClampFloat(Float=&Add_Float(A=-15, B=&Multiply_Float(A=10, B=$User.NumberKills)), Max=30)
   ) {isSetParameters=true}
   &SpawnRate(Spawn Probability=$Emitter.SpawnProbabilityMath, SpawnRate=$Emitter.SpawnRateMath, Use Spawn Probability=true)
  ]
  particleSpawnScript [
   &InitializeParticle(
    Color=<rgba 0.8317 0 2 1>
    Lifetime=&UniformRangedFloat(Minimum=0.2, Maximum=0.5)
    Module.Write Color=true
    Module.Write Lifetime=true
    Module.Write Mass=true
    Module.Write Material Random=true
    Module.Write Position=false
    Module.Write Scale=false
    Module.Write SpriteRotation=true
    Module.Write SpriteSize=true
    Position=&SimulationPosition()
    Sprite Rotation=&UniformRangedFloat(Minimum=0, Maximum=360)
    Sprite Size=&Vector2DFromFloat(Value=&UniformRangedFloat(Minimum=6, Maximum=17))
   )
   &DynamicMaterialParameters(
    Index 0 Param 1=&UniformRangedFloat(Minimum=0, Maximum=1)
    Index 0 Param 2=&UniformRangedFloat(Minimum=0.1, Maximum=1)
    Index 0 Param 3=1
    Module.Param0WriteEnabled=true
    Module.Param1WriteEnabled=true
    Module.Param2WriteEnabled=true
    Module.Param3WriteEnabled=false
   ) {module=DynamicMaterialParameters001}
   /SetVariables(
    Particles.StableRandom:NiagaraFloat=&UniformRangedFloat(Minimum=0, Maximum=1)
    Particles.SubImageIndex=&UniformAOrBFloat(A=1, B=3)
   ) {isSetParameters=true}
  ]
  particleUpdateScript [
   &ParticleState()
   &SkeletalMeshLocation(
    Bone Sampling Mode=::NewEnumerator4
    Inherit Velocity=false
    MaskSpawnGroup=false
    Sampled Position Offset=&Lerp_Vector(A=<0 -3 0>, B=<0 0 0>, Alpha=$Particles.StableRandom)
    Sampled Position Offset Coordinate Space:ENiagaraMeshTransforms=::NewEnumerator2
    Skeletal Mesh=$User.SkeletalMesh
   )
   &DynamicMaterialParameters(
    Index 0 Param 3=&FloatFromCurve(FloatCurve=DataInterface() {propertyValues {Curve {keys [<at 0 1 RCIM_Cubic>,<at 0.2 0.025 RCIM_Cubic>,<at 1 1>]}}})
    Module.Param0WriteEnabled=false
    Module.Param1WriteEnabled=false
    Module.Param2WriteEnabled=true
    Module.Param3WriteEnabled=false
   )
   &ScaleColor(
    Scale RGB=&Multiply_VectorByFloat(
     Vector=&MakeVectorFromLinearColorRGB(Color=&ColorFromCurve(ColorCurve=DataInterface() {propertyValues {
         AlphaCurve {keys []}
         BlueCurve {keys [<at 0 1>,<at 0.15 10 RCIM_Cubic>,<at 1 1 RCIM_Cubic>]}
         GreenCurve {keys [<at 0 1>, {arriveTangent=-0.406, interpMode=RCIM_Cubic, leaveTangent=-0.406, time=0.15, value=9.045}, <at 1 0.594 RCIM_Cubic>]}
         RedCurve {keys [<at 0 1>, {arriveTangent=-0.4036, interpMode=RCIM_Cubic, leaveTangent=-0.4036, time=0.15, value=8.5417}, <at 1 0.5964 RCIM_Cubic>]}
        }}))
     Float=1
    )
    ScaleA=false
    ScaleRGB=true
   )
   /Niagara/Modules/Update/SubUV/SubUVAnimation(
    Number Of Frames=6
    SubUV Lookup Index=&Add_Float(A=&Multiply_Float(A=0.5, B=$Particles.NormalizedAge), B=$Particles.Initial.SubImageIndex)
   )
  ]
  rendererClasses [&NiagaraSpriteRendererProperties]
  renderers [{rendererClass=&NiagaraSpriteRendererProperties}]
 }
 Lights {
  emitterUpdateScript [
   &EmitterState()
   &SpawnRate(
    SpawnRate=&ClampFloat(Float=&Multiply_Float(A=2, B=$User.NumberKills), Max=6)
    Use Spawn Probability=false
   )
  ]
  particleSpawnScript [
   /SetVariables(Particles.KillsRamp:NiagaraFloat=&ClampFloat(Float=&Multiply_Float(A=0.3333, B=$User.NumberKills))) {isSetParameters=true}
   &InitializeParticle(
    Color=&MultiplyLinearColor(
     Linear Color=&UniformRangedLinearColor(Minimum=<rgba 1 0 0.7667 1>, Maximum=<rgba 0.5167 0 1 1>)
     Scale Factor=&MakeVector4_FromFloat(Float=&ClampFloat(
       Float=&Lerp_Float(A=-5, B=20, Alpha=&Multiply_Float(A=$Particles.KillsRamp, B=$Particles.KillsRamp))
       Max=50
      ))
    )
    Lifetime=&Lerp_Float(
     A=&UniformRangedFloat(Minimum=2, Maximum=2)
     B=&UniformRangedFloat(Minimum=0.15, Maximum=0.5)
     Alpha=$Particles.KillsRamp
    )
    Mesh Scale=&VectorFromFloat(Value=0.5)
    Module.Write Color=true
    Module.Write Lifetime=true
    Module.Write Mass=false
    Module.Write Material Random=true
    Module.Write Position=false
    Module.Write Scale=false
    Module.Write SpriteRotation=false
    Module.Write SpriteSize=false
    Position=&SimulationPosition()
    Sprite Size=<500 500>
   )
   /SetVariables(Particles.LightExponent=2) {isSetParameters=true}
  ]
  particleUpdateScript [
   &ParticleState()
   /SetVariables(Particles.KillsRamp:NiagaraFloat=&ClampFloat(Float=&Multiply_Float(A=0.3333, B=$User.NumberKills))) {isSetParameters=true}
   /SetVariables(Particles.LightRadius=&Lerp_Float(A=8, B=10, Alpha=$Particles.KillsRamp)) {isSetParameters=true}
   &SkeletalMeshLocation(
    Bone Sampling Mode=::NewEnumerator4
    Inherit Velocity=false
    MaskSpawnGroup=false
    Sampled Position Offset=<0 -1.4 0>
    Sampled Position Offset Coordinate Space:ENiagaraMeshTransforms=::NewEnumerator2
    Skeletal Mesh=$User.SkeletalMesh
   )
   &ScaleColor(
    Scale RGB=&Multiply_VectorByFloat(
     Vector=/Niagara/DynamicInputs/Math/LerpMultipleVectors(
      Alpha:NiagaraFloat=0
      V 1:Vector3f=<1 1 1>
      V 2:Vector3f=<1 1 1>
      V 3:Vector3f=&MakeVectorFromLinearColorRGB(Color=<rgba 0.8017 0.42 5 1>)
      V 4:Vector3f=?
      V 5:Vector3f=?
      Vector Count:ENiagara_MultipleLerpCount=::NewEnumerator0
     )
     Float=&Lerp_Float(
      A=&FloatFromCurve(FloatCurve=DataInterface() {propertyValues {Curve {keys [
           {arriveTangent=-0.0218, interpMode=RCIM_Cubic, time=0, value=0}
           {arriveTangent=0.0444, interpMode=RCIM_Cubic, leaveTangent=0.0444, tangentMode=RCTM_User, time=0.25, value=1}
           {arriveTangent=-0.0018, interpMode=RCIM_Cubic, leaveTangent=-0.0018, tangentMode=RCTM_User, time=0.75, value=1}
           {arriveTangent=-1.3928, interpMode=RCIM_Cubic, leaveTangent=-1.3928, tangentMode=RCTM_User, time=1, value=0}
          ]}}})
      B=&FloatFromCurve(
       FloatCurve=DataInterface() {propertyValues {Curve {keys [
           {arriveTangent=4.8231, interpMode=RCIM_Cubic, leaveTangent=4.8231, tangentMode=RCTM_User, time=0, value=0}
           {interpMode=RCIM_Cubic, tangentMode=RCTM_User, time=0.3294, value=1}
           {arriveTangent=-1.3928, interpMode=RCIM_Cubic, leaveTangent=-1.3928, tangentMode=RCTM_User, time=1, value=0}
          ]}}}
       Scale Curve=5
      )
      Alpha=$Particles.KillsRamp
     )
    )
    ScaleA=false
    ScaleRGB=true
   )
  ]
  rendererClasses [&NiagaraLightRendererProperties]
  renderers [{rendererClass=&NiagaraLightRendererProperties, rendererIndex=0}]
 }
 Runes {
  emitterUpdateScript [
   &EmitterState()
   &SpawnRate(
    Spawn Probability=0.2
    SpawnRate=&UniformRangedFloat(
     Minimum=0
     Maximum=&ClampFloat(Float=&Add_Float(A=&Multiply_Float(A=$User.NumberKills, B=24.5), B=-23.5), Max=50)
    )
    Use Spawn Probability=true
   )
  ]
  particleSpawnScript [
   &InitializeParticle(
    Color=&MakeLinearColor(
     R:NiagaraFloat=1
     G:NiagaraFloat=&UniformRangedFloat(Minimum=0.5, Maximum=3)
     B:NiagaraFloat=&UniformRangedFloat(Minimum=1, Maximum=10)
     A:NiagaraFloat=&UniformRangedFloat(Minimum=0.5, Maximum=1)
    )
    Lifetime=&UniformRangedFloat(Minimum=0.1, Maximum=&UniformRangedFloat(Minimum=0.2, Maximum=1))
    Module.Write Color=true
    Module.Write Lifetime=true
    Module.Write Mass=false
    Module.Write Material Random=true
    Module.Write Position=false
    Module.Write Scale=false
    Module.Write SpriteRotation=true
    Module.Write SpriteSize=true
    Position=&SimulationPosition()
    Sprite Rotation=&UniformRangedFloat(Minimum=0, Maximum=360)
    Sprite Size=&Vector2DFromFloat(Value=&UniformRangedFloat(Minimum=2, Maximum=4))
   )
   /SetVariables(Particles.SpriteInitRotationRate:NiagaraFloat=&Multiply_Float(
     A=&UniformAOrBFloat(A=-1)
     B=&UniformRangedFloat(Minimum=5, Maximum=10)
    )) {isSetParameters=true}
   &SphereLocation(
    MaskSpawnGroup=false
    Override Seed=false
    Sphere Radius=10
    Surface Only=false
    UseNonUniformScale=false
    UseOffset=false
    UseSphereOrientation=false
   )
   &AddVelocity(Velocity=&Multiply_VectorByFloat(Vector=$Output.SphereLocation.SphereNormal, Float=&UniformRangedFloat(Minimum=200, Maximum=400)))
   &SubUVAnimation(SubUV Animation Mode=::NewEnumerator1)
  ]
  particleUpdateScript [
   &ParticleState()
   &SpriteRotationRate(Rotation Rate=&Multiply_Float(A=$Particles.SpriteInitRotationRate, B=&VectorLength(Vector=$Particles.Velocity)))
   /SetVariables(Particles.Velocity=&Multiply_VectorByFloat(
     Vector=$Particles.Initial.Velocity
     Float=&FloatFromCurve(FloatCurve=DataInterface() {propertyValues {Curve {keys [
          {arriveTangent=-5.5773, interpMode=RCIM_Cubic, leaveTangent=-5.5773, tangentMode=RCTM_User, time=0, value=1}
          {arriveTangent=-0.9831, interpMode=RCIM_Cubic, leaveTangent=-0.9831, tangentMode=RCTM_User, time=0.2018, value=0.1945}
          <at 0.75 0>
         ]}}})
    )) {isSetParameters=true}
   &SolveForcesAndVelocity(Clamp Velocity=false, Limit Acceleration=false)
   &ScaleColor(
    Scale Alpha=&FloatFromCurve(FloatCurve=DataInterface() {propertyValues {Curve {keys [{interpMode=RCIM_Cubic, tangentMode=RCTM_User, time=0.75, value=1}, <at 1 0.5>]}}})
    Scale RGB=&MakeVector(
     X=&UniformAOrBFloat(A=0)
     Y=1
     Z=&FloatFromCurve(FloatCurve=DataInterface() {propertyValues {Curve {keys [
          {arriveTangent=-48.3794, interpMode=RCIM_Cubic, leaveTangent=-48.3794, tangentMode=RCTM_User, time=0.2, value=5}
          {arriveTangent=-0.0507, interpMode=RCIM_Cubic, leaveTangent=-0.0508, tangentMode=RCTM_User, time=0.32, value=1}
          {arriveTangent=90.7103, interpMode=RCIM_Cubic, leaveTangent=90.7105, tangentMode=RCTM_User, time=1, value=20}
         ]}}})
    )
    ScaleA=true
    ScaleRGB=true
   ) {module=ScaleColor001}
   &ScaleSpriteSize(Scale Factor=&Vector2DFromFloat(Value=&FloatFromCurve(FloatCurve=DataInterface() {propertyValues {Curve {keys [
          {arriveTangent=11.6143, interpMode=RCIM_Cubic, leaveTangent=11.6143, tangentMode=RCTM_User, time=0, value=0.25}
          {arriveTangent=1.5, interpMode=RCIM_Cubic, leaveTangent=1.5, time=0.1, value=1.5}
          {interpMode=RCIM_Cubic, leaveTangent=0.8761, time=0.5, value=1}
         ]}}})))
  ]
  rendererClasses [&NiagaraSpriteRendererProperties]
  renderers [{rendererClass=&NiagaraSpriteRendererProperties}]
 }
}
summary {
 systemName=NS_Backpack_CloakedAssassin
 emitters [
  {emitterName=Lights, rendererClasses [&NiagaraLightRendererProperties]}
  {emitterName=Lightning, rendererClasses [&NiagaraSpriteRendererProperties]}
  {
   emitterName=BigLightningRibbon
   rendererClasses [&NiagaraRibbonRendererProperties, &NiagaraSpriteRendererProperties]
  }
  {emitterName=BigLightningEndPicker, rendererClasses [&NiagaraSpriteRendererProperties]}
  {emitterName=Runes, rendererClasses [&NiagaraSpriteRendererProperties]}
 ]
 userVariables [
  User.NumberKills {type=NiagaraFloat, defaultValue=0, description=""}
  User.SkeletalMesh {
   type=NiagaraDataInterfaceSkeletalMesh
   defaultValue {
    struct=NiagaraExt_VariableValue_DataInterface
    value {dataInterface=&self:NiagaraDataInterfaceSkeletalMesh_0}
   }
  }
 ]
}
