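# Niagara on-kill effect for the Cloaked Assassin backpack cosmetic, spawned around a bone or socket on the User.SkeletalMesh.
# BurstRunesLocal throws spinning, decelerating rune sprites, with a second sprite renderer and custom Ray attributes for streaks on some particles.
# RateRunesWorld adds a tapering stream of similar runes.
# Lightning sprites and a purple light pulse resample the mesh each update to stay attached, animated by material-parameter, color and radius changes.
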NiagaraSystem /BRCosmetics/Accessories/FORT_Backpacks/Backpack_F_MED_Cloaked_Assassin/Effects/NS_Backpack_CloakedAssassin_OnKill
dependencies {
 usedDataInterfaces [&NiagaraDataInterfaceCurve, &NiagaraDataInterfaceColorCurve]
 usedDynamicInputs []
 usedModules [&SystemState]
}
emitters {
 BurstRunesLocal {
  emitterUpdateScript [
   &EmitterState()
   &SpawnBurst_Instantaneous(Spawn Count=20, Use Spawn Probability=false)
  ]
  particleSpawnScript [
   /Niagara/Modules/Spawn/Initialization/InitializeParticle(
    Color=&MakeLinearColor(
     R:NiagaraFloat=1
     G:NiagaraFloat=&UniformRangedFloat(Minimum=1, Maximum=20)
     B:NiagaraFloat=&UniformRangedFloat(Minimum=1, Maximum=10)
     A:NiagaraFloat=&UniformRangedFloat(Minimum=0.5, Maximum=1)
    )
    Lifetime=&UniformRangedFloat(Minimum=0.2, Maximum=0.5)
    Module.Write Color=true
    Module.Write Lifetime=true
    Module.Write Mass=false
    Module.Write Material Random=true
    Module.Write Position=false
    Module.Write Scale=false
    Module.Write SpriteRotation=true
    Module.Write SpriteSize=true
    Position=&SimulationPosition()
    Sprite Rotation=&UniformRangedFloat(Minimum=0, Maximum=360)
    Sprite Size=&Vector2DFromFloat(Value=&UniformRangedFloat(Minimum=1, Maximum=5))
   )
   /SetVariables(Particles.SpriteInitRotationRate:NiagaraFloat=&Multiply_Float(
     A=&UniformAOrBFloat(A=-1)
     B=&UniformRangedFloat(Minimum=5, Maximum=10)
    )) {isSetParameters=true, module=SetVariables_#1}
   &AddVelocity(Velocity=&Multiply_VectorByFloat(Vector=&UniformRangedVector(Minimum=<-3 -8 -3>, Maximum=<3 -0.5 3>), Float=100))
   /SetVariables(Particles.ZeroOrOne:NiagaraFloat=&UniformAOrBFloat(A=0, Distribution Weight=0.75)) {isSetParameters=true, module=SetVariables_#2}
   /SetVariables(
    Particles.RayColor:LinearColor=&UniformRangedLinearColor(Minimum=<rgba 0.3115 0.2188 3 0.75>, Maximum=<rgba 10 0 3.5992 1>, Color Channel Mode=::NewEnumerator1)
    Particles.RaySize:Vector2f=&Multiply_Vector2(
     A=&UniformRangedVector2D(Minimum=<15 10>, Maximum=<25 50>)
     B=&Vector2DFromFloat(Value=$Particles.ZeroOrOne)
    )
    Particles.RayVelocity:Vector3f=&Multiply_Vector(A=$Particles.Velocity, B=<0.125 1 0.125>)
   ) {isSetParameters=true}
   /Niagara/Modules/Update/SubUV/SubUVAnimation(SubUV Animation Mode=::NewEnumerator1)
   &SkeletalMeshLocation(
    Bone / Socket Index=1
    Bone Sampling Mode=::NewEnumerator4
    Inherit Velocity=false
    MaskSpawnGroup=false
    Sampled Position Offset=<0 -3.5 0>
    Sampled Position Offset Coordinate Space:ENiagaraMeshTransforms=::NewEnumerator3
    Skeletal Mesh=$User.SkeletalMesh
   )
   /SetVariables(Particles.RayPos:Vector3f=&Add_Vector(A=Linked() {linkedVariable {name=Particles.Position, type=NiagaraPosition}}, B=<0 5 0>)) {isSetParameters=true}
  ]
  particleUpdateScript [
   &ParticleState()
   /SetVariables(Particles.AdjustedNormalizedAge:NiagaraFloat=&ClampFloat(Float=&Multiply_Float(A=$Particles.NormalizedAge, B=1.45))) {isSetParameters=true}
   &SpriteRotationRate(Rotation Rate=&Multiply_Float(A=$Particles.SpriteInitRotationRate, B=&VectorLength(Vector=$Particles.Velocity)))
   &ScaleVelocity(Velocity Scale=&Add_Vector(
     A=&Multiply_VectorByFloat(
      Vector=<0 1 0>
      Float=&FloatFromCurve(
       FloatCurve=DataInterface() {propertyValues {Curve {keys [
           {arriveTangent=-5.3117, interpMode=RCIM_Cubic, leaveTangent=-5.3117, tangentMode=RCTM_User, time=0, value=1}
           {arriveTangent=-0.008, interpMode=RCIM_Cubic, leaveTangent=-0.008, tangentMode=RCTM_User, time=0.5, value=0}
          ]}}}
       CurveIndex=$Particles.AdjustedNormalizedAge
      )
     )
     B=&Multiply_VectorByFloat(
      Vector=<1 0 1>
      Float=&FloatFromCurve(
       FloatCurve=DataInterface() {propertyValues {Curve {keys [
           {arriveTangent=-3.0841, interpMode=RCIM_Cubic, leaveTangent=-3.0841, tangentMode=RCTM_User, time=0, value=1}
           {arriveTangent=0.0234, interpMode=RCIM_Cubic, leaveTangent=0.0234, tangentMode=RCTM_User, time=0.5, value=0}
          ]}}}
       CurveIndex=$Particles.AdjustedNormalizedAge
      )
     )
    ))
   &SolveForcesAndVelocity(Clamp Velocity=false, Limit Acceleration=false)
   &ScaleColor(
    Scale Alpha=&FloatFromCurve(
     FloatCurve=DataInterface() {propertyValues {Curve {keys [{interpMode=RCIM_Cubic, tangentMode=RCTM_User, time=0.75, value=1}, <at 1 0.5>]}}}
     CurveIndex=$Particles.AdjustedNormalizedAge
    )
    Scale RGB=&MakeVector(
     X=&UniformAOrBFloat(A=0)
     Y=1
     Z=&FloatFromCurve(
      FloatCurve=DataInterface() {propertyValues {Curve {keys [
          {arriveTangent=-172.0061, interpMode=RCIM_Cubic, leaveTangent=-172.0061, tangentMode=RCTM_User, time=0.2, value=10}
          {arriveTangent=-0.0507, interpMode=RCIM_Cubic, leaveTangent=-0.0508, tangentMode=RCTM_User, time=0.32, value=1}
          {arriveTangent=90.7103, interpMode=RCIM_Cubic, leaveTangent=90.7105, tangentMode=RCTM_User, time=1, value=20}
         ]}}}
      CurveIndex=$Particles.AdjustedNormalizedAge
     )
    )
    ScaleA=true
    ScaleRGB=true
   )
   /SetVariables(
    Particles.RayColor:LinearColor=&MultiplyLinearColor(
     Linear Color=$Particles.Initial.RayColor
     Scale Factor=&MakeFloat4FromVectorAndFloat(
      "Vector (RGB)"=&Lerp_Vector(A=<1 1 1>, B=<0.1 0.1 5>, Alpha=$Particles.NormalizedAge)
      "Float (Alpha)"=&FloatFromCurve(FloatCurve=DataInterface() {propertyValues {Curve {keys [
           {arriveTangent=-3.0855, interpMode=RCIM_Cubic, leaveTangent=-3.0855, tangentMode=RCTM_User, time=0, value=1}
           {arriveTangent=-0.5556, interpMode=RCIM_Cubic, leaveTangent=-0.5556, tangentMode=RCTM_User, time=0.3643, value=0.2598}
           {arriveTangent=-0.2827, interpMode=RCIM_Cubic, leaveTangent=-0.2827, tangentMode=RCTM_User, time=1, value=0}
          ]}}})
     )
    )
    Particles.RaySize:Vector2f=&Multiply_Vector2(
     A=$Particles.Initial.RaySize
     B=&MakeVector2D(
      X=&Lerp_Float(A=1, B=1.2, Alpha=$Particles.NormalizedAge)
      Y=&Lerp_Float(A=1, B=0.85, Alpha=$Particles.NormalizedAge)
     )
    )
   ) {isSetParameters=true}
  ]
  rendererClasses [&NiagaraSpriteRendererProperties]
  renderers [
   {rendererClass=&NiagaraSpriteRendererProperties}
   {rendererClass=&NiagaraSpriteRendererProperties, rendererIndex=1}
  ]
 }
 Lightning001 {
  emitterUpdateScript [
   &EmitterState()
   &SpawnBurst_Instantaneous(Spawn Count=5, Use Spawn Probability=false)
  ]
  particleSpawnScript [
   &InitializeParticle(
    Color=<rgba 0.8317 0 2 1>
    Lifetime=&UniformRangedFloat(Minimum=0.2, Maximum=1)
    Module.Write Color=true
    Module.Write Lifetime=true
    Module.Write Mass=true
    Module.Write Material Random=true
    Module.Write Position=false
    Module.Write Scale=false
    Module.Write SpriteRotation=true
    Module.Write SpriteSize=true
    Position=&SimulationPosition()
    Sprite Rotation=&UniformRangedFloat(Minimum=0, Maximum=360)
    Sprite Size=&Vector2DFromFloat(Value=&Lerp_Float(A=7, B=25, Alpha=&UniformRangedFloat(Minimum=0, Maximum=1)))
   )
   &DynamicMaterialParameters(
    Index 0 Param 1=&UniformRangedFloat(Minimum=0, Maximum=1)
    Index 0 Param 2=&UniformRangedFloat(Minimum=0.1, Maximum=1)
    Index 0 Param 3=1
    Module.Param0WriteEnabled=true
    Module.Param1WriteEnabled=true
    Module.Param2WriteEnabled=true
    Module.Param3WriteEnabled=false
   ) {module=DynamicMaterialParameters001}
   /SetVariables(Particles.SubImageIndex=&UniformRangedFloat(Minimum=0, Maximum=3)) {isSetParameters=true}
  ]
  particleUpdateScript [
   &SkeletalMeshLocation(
    Bone / Socket Index=1
    Bone Sampling Mode=::NewEnumerator4
    Inherit Velocity=false
    MaskSpawnGroup=false
    Sampled Position Offset=<0 -5 0>
    Sampled Position Offset Coordinate Space:ENiagaraMeshTransforms=::NewEnumerator3
    Skeletal Mesh=$User.SkeletalMesh
   )
   &ParticleState()
   &DynamicMaterialParameters(
    Index 0 Param 3=&FloatFromCurve(FloatCurve=DataInterface() {propertyValues {Curve {keys [<at 0 1 RCIM_Cubic>,<at 0.2 0.025 RCIM_Cubic>,<at 1 1>]}}})
    Module.Param0WriteEnabled=false
    Module.Param1WriteEnabled=false
    Module.Param2WriteEnabled=true
    Module.Param3WriteEnabled=false
   )
   &ScaleColor(
    Scale RGB=&Multiply_VectorByFloat(
     Vector=&MakeVectorFromLinearColorRGB(Color=&ColorFromCurve(ColorCurve=DataInterface() {propertyValues {
         AlphaCurve {keys []}
         BlueCurve {keys [<at 0 1>,<at 0.15 10 RCIM_Cubic>,<at 1 1 RCIM_Cubic>]}
         GreenCurve {keys [<at 0 1>, {arriveTangent=-0.1823, interpMode=RCIM_Cubic, leaveTangent=-0.1823, time=0.15, value=9.045}, <at 1 0.8177 RCIM_Cubic>]}
         RedCurve {keys [<at 0 1>, {arriveTangent=-0.1812, interpMode=RCIM_Cubic, leaveTangent=-0.1812, time=0.15, value=8.5417}, <at 1 0.8188 RCIM_Cubic>]}
        }}))
     Float=1
    )
    ScaleA=false
    ScaleRGB=true
   )
   &SubUVAnimation(
    Number Of Frames=6
    SubUV Lookup Index=&Add_Float(A=&Multiply_Float(A=0.5, B=$Particles.NormalizedAge), B=$Particles.Initial.SubImageIndex)
   )
  ]
  rendererClasses [&NiagaraSpriteRendererProperties]
  renderers [{rendererClass=&NiagaraSpriteRendererProperties}]
 }
 Lights {
  emitterUpdateScript [
   &EmitterState()
   &SpawnBurst_Instantaneous(Spawn Count=1, Use Spawn Probability=false)
  ]
  particleSpawnScript [
   &InitializeParticle(
    Color=<rgba 2.7086 0.565 5 1>
    Lifetime=&UniformRangedFloat(Minimum=1.25, Maximum=1.5)
    Mesh Scale=&VectorFromFloat(Value=0.5)
    Module.Write Color=true
    Module.Write Lifetime=true
    Module.Write Mass=false
    Module.Write Material Random=true
    Module.Write Position=false
    Module.Write Scale=true
    Module.Write SpriteRotation=false
    Module.Write SpriteSize=true
    Position=&SimulationPosition()
    Sprite Size=<500 500>
   )
   /SetVariables(Particles.LightExponent=1) {isSetParameters=true}
   /SetVariables() {isSetParameters=true}
  ]
  particleUpdateScript [
   &ParticleState()
   /SetVariables(Particles.LightRadius=&ClampFloat(Float=&Lerp_Float(A=0, B=70, Alpha=$Particles.NormalizedAge), Max=12)) {isSetParameters=true}
   &SkeletalMeshLocation(
    Bone / Socket Index=1
    Bone Sampling Mode=::NewEnumerator4
    Inherit Velocity=false
    MaskSpawnGroup=false
    Sampled Position Offset=<0 -3.5 0>
    Sampled Position Offset Coordinate Space:ENiagaraMeshTransforms=::NewEnumerator3
    Skeletal Mesh=$User.SkeletalMesh
   )
   &ScaleColor(
    Scale RGB=&Multiply_VectorByFloat(
     Vector=&Lerp_Vector(
      A=<1 1 1>
      B=&MakeVectorFromLinearColorRGB(Color=<rgba 2.2317 1.98 5 1>)
      Alpha=$Particles.NormalizedAge
     )
     Float=&FloatFromCurve(FloatCurve=DataInterface() {propertyValues {Curve {keys [
          {arriveTangent=14.6944, interpMode=RCIM_Cubic, leaveTangent=14.6944, tangentMode=RCTM_User, time=0, value=0}
          {interpMode=RCIM_Cubic, tangentMode=RCTM_User, time=0.1, value=1}
          {arriveTangent=-1.3928, interpMode=RCIM_Cubic, leaveTangent=-1.3928, tangentMode=RCTM_User, time=1, value=0}
         ]}}})
    )
    ScaleA=false
    ScaleRGB=true
   )
  ]
  rendererClasses [&NiagaraLightRendererProperties]
  renderers [{rendererClass=&NiagaraLightRendererProperties, rendererIndex=0}]
 }
 RateRunesWorld {
  emitterUpdateScript [
   &EmitterState()
   &SpawnRate(SpawnRate=&Lerp_Float(A=50, B=1, Alpha=$Emitter.NormalizedLoopAge), Use Spawn Probability=false)
  ]
  particleSpawnScript [
   &InitializeParticle(
    Color=&MakeLinearColor(
     R:NiagaraFloat=1
     G:NiagaraFloat=&UniformRangedFloat(Minimum=1, Maximum=20)
     B:NiagaraFloat=&UniformRangedFloat(Minimum=1, Maximum=10)
     A:NiagaraFloat=&UniformRangedFloat(Minimum=0.5, Maximum=1)
    )
    Lifetime=&UniformRangedFloat(Minimum=0.2, Maximum=0.75)
    Module.Write Color=true
    Module.Write Lifetime=true
    Module.Write Mass=false
    Module.Write Material Random=true
    Module.Write Position=false
    Module.Write Scale=false
    Module.Write SpriteRotation=true
    Module.Write SpriteSize=true
    Position=&SimulationPosition()
    Sprite Rotation=&UniformRangedFloat(Minimum=0, Maximum=360)
    Sprite Size=&Vector2DFromFloat(Value=&UniformRangedFloat(Minimum=1, Maximum=5))
   )
   /SetVariables(Particles.SpriteInitRotationRate:NiagaraFloat=&Multiply_Float(
     A=&UniformAOrBFloat(A=-1)
     B=&UniformRangedFloat(Minimum=5, Maximum=10)
    )) {isSetParameters=true, module=SetVariables_#1}
   &AddVelocity(Velocity=&Multiply_VectorByFloat(Vector=&UniformRangedVector(Minimum=<-0.5 -2 -0.5>, Maximum=<0.5 -0.5 0.5>), Float=100))
   /SetVariables(Particles.ZeroOrOne:NiagaraFloat=&UniformAOrBFloat(A=0, Distribution Weight=0.25)) {isSetParameters=true, module=SetVariables_#2}
   &SubUVAnimation(SubUV Animation Mode=::NewEnumerator1)
   &SkeletalMeshLocation(
    Bone / Socket Index=1
    Bone Sampling Mode=::NewEnumerator4
    Inherit Velocity=false
    MaskSpawnGroup=false
    Sampled Position Offset=<0 -3.5 0>
    Sampled Position Offset Coordinate Space:ENiagaraMeshTransforms=::NewEnumerator3
    Skeletal Mesh=$User.SkeletalMesh
   )
  ]
  particleUpdateScript [
   &ParticleState()
   &SpriteRotationRate(Rotation Rate=&Multiply_Float(A=$Particles.SpriteInitRotationRate, B=&VectorLength(Vector=$Particles.Velocity)))
   /SetVariables(Particles.Velocity=&Multiply_VectorByFloat(
     Vector=$Particles.Initial.Velocity
     Float=&FloatFromCurve(FloatCurve=DataInterface() {propertyValues {Curve {keys [
          {arriveTangent=-5.3117, interpMode=RCIM_Cubic, leaveTangent=-5.3117, tangentMode=RCTM_User, time=0, value=1}
          {arriveTangent=-0.008, interpMode=RCIM_Cubic, leaveTangent=-0.008, tangentMode=RCTM_User, time=0.5, value=0}
         ]}}})
    )) {isSetParameters=true}
   &SolveForcesAndVelocity(Clamp Velocity=false, Limit Acceleration=false)
   &ScaleColor(
    Scale Alpha=&FloatFromCurve(FloatCurve=DataInterface() {propertyValues {Curve {keys [{interpMode=RCIM_Cubic, tangentMode=RCTM_User, time=0.75, value=1}, <at 1 0.5>]}}})
    Scale RGB=&MakeVector(
     X=&UniformAOrBFloat(A=0)
     Y=1
     Z=&FloatFromCurve(FloatCurve=DataInterface() {propertyValues {Curve {keys [
          {arriveTangent=-43.7988, interpMode=RCIM_Cubic, leaveTangent=-43.7988, tangentMode=RCTM_User, time=0, value=10}
          {arriveTangent=-0.0507, interpMode=RCIM_Cubic, leaveTangent=-0.0508, tangentMode=RCTM_User, time=0.2, value=1}
          {arriveTangent=58.8836, interpMode=RCIM_Cubic, leaveTangent=58.8836, tangentMode=RCTM_User, time=1, value=20}
         ]}}})
    )
    ScaleA=true
    ScaleRGB=true
   )
  ]
  rendererClasses [&NiagaraSpriteRendererProperties]
  renderers [{rendererClass=&NiagaraSpriteRendererProperties}]
 }
}
summary {
 systemName=NS_Backpack_CloakedAssassin_OnKill
 emitters [
  {emitterName=Lights, rendererClasses [&NiagaraLightRendererProperties]}
  {emitterName=BurstRunesLocal, rendererClasses [&NiagaraSpriteRendererProperties]}
  {emitterName=Lightning001, rendererClasses [&NiagaraSpriteRendererProperties]}
  {emitterName=RateRunesWorld, rendererClasses [&NiagaraSpriteRendererProperties]}
 ]
 userVariables [User.SkeletalMesh {
   type=NiagaraDataInterfaceSkeletalMesh
   defaultValue {
    struct=NiagaraExt_VariableValue_DataInterface
    value {dataInterface=&self:NiagaraDataInterfaceSkeletalMesh_0}
   }
  }]
}
