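# Animation Blueprint for the Cloaked Assassin back bling, spinning its DYN_Cube bone procedurally with springy secondary motion on the master cube bone.
# The event graph derives slow and fast roll/yaw rotators from real time each update.
# Grounded poses blend from slow spin to fast spin plus AnimDynamics when moving, then add a pack offset and more dynamics.
# Jumping or falling blends to a separate fast-spin branch with two stacked linear-spring AnimDynamics nodes.
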AnimBlueprint /BRCosmetics/Accessories/FORT_Backpacks/Backpack_F_MED_Cloaked_Assassin/Meshes/F_MED_Cloaked_Assassin_Pack_AnimBP
target_skeleton=&/F_MED_Cloaked_Assassin_Pack_Skeleton
graphs [
 AnimGraph {nodes [
   Blend Poses by bool() {blendType=Linear}
   n1 = "Transform (Modify) Bone - Bone: DYN_Cube"(
    ComponentPose=Local To Component(LocalPose=Local Space Ref Pose())
    Alpha=1
    Translation="0,0,0"
    Rotation=Get Spinning_Rotator_SPD2() -> (Spinning_Rotator_SPD2:Rotator)
    Scale="1,1,1"
   ) {rotationMode=BMM_Additive, rotationSpace=BCS_BoneSpace, alphaBoolBlend {blendOption=Linear}, boneToModify {boneName=DYN_Cube}}
   n2 = Anim Dynamics - Bone: DYN_MSTR_Cube(ComponentPose=@n1, Alpha=0.5, ExternalForce="0,0,0") {
    angularDampingOverride=0.9
    angularSpringConstant=150
    bGravityOverrideInSimSpace=true
    bLinearSpring=true
    bOverrideAngularDamping=true
    bOverrideLinearDamping=true
    gravityScale=0
    linearDampingOverride=0.999
    linearSpringConstant=50
    simulationSpace=BoneRelative
    alphaBoolBlend {blendOption=Linear}
    boundBone {boneName=DYN_MSTR_Cube}
    componentAppliedLinearAccClamp=<6000 6000 6000>
    componentLinearAccScale=<0.3 0.3 0.3>
    componentLinearVelScale=<0.6 0.6 0.6>
    physicsBodyDefinitions [{
      sphereCollisionRadius=0
      boundBone {boneName=DYN_MSTR_Cube}
      boxExtents=<50 50 50>
      constraintSetup {angularLimitsMax=<0 0 0>, angularLimitsMin=<0 0 0>, angularTarget=<0 0 0>, linearAxesMax=<3 0 3>, linearAxesMin=<-3 0 -3>}
      localJointOffset=<0 -3 0>
     }]
    relativeSpaceBone {boneName=root}
   }
   n3 = Component To Local(ComponentPose=@n2)
   n4 = "Transform (Modify) Bone - Bone: DYN_Cube"(
    ComponentPose=Local To Component(LocalPose=Local Space Ref Pose())
    Alpha=1
    Translation="0,0,0"
    Rotation=Get Spinning_Rotator() -> (Spinning_Rotator:Rotator)
    Scale="1,1,1"
   ) {rotationMode=BMM_Additive, rotationSpace=BCS_BoneSpace, alphaBoolBlend {blendOption=Linear}, boneToModify {boneName=DYN_Cube}}
   n5 = Component To Local(ComponentPose=@n4)
   n6 = OR Boolean!(A=Get bIsSprinting(), B=Get bIsAccelerating2D(), C:bool=Get bIsMoving2D()) -> (ReturnValue)
   n7 = "Transform (Modify) Bone - Bone: pack"(
    ComponentPose=Local To Component(LocalPose=Blend Poses by bool(BlendPose_0=@n3, BlendPose_1=@n5, BlendTime_0=1.5, BlendTime_1=3.5, bActiveValue=@n6) {blendType=Linear})
    Translation="0.000,-0.5,0.000"
    Rotation="0,0,0"
   ) {
    rotationMode=BMM_Additive
    rotationSpace=BCS_BoneSpace
    translationMode=BMM_Additive
    translationSpace=BCS_ParentBoneSpace
    alphaBoolBlend {blendOption=Linear}
    boneToModify {boneName=pack}
    translation {y=-0.5}
   }
   n8 = Anim Dynamics - Bone: DYN_MSTR_Cube(ComponentPose=@n7, Alpha=0.9, ExternalForce="0,0,0") {
    angularDampingOverride=0.9
    angularSpringConstant=150
    bAngularSpring=true
    bGravityOverrideInSimSpace=true
    bLinearSpring=true
    bOverrideAngularDamping=true
    bOverrideLinearDamping=true
    gravityScale=0
    linearDampingOverride=0.995
    linearSpringConstant=50
    simulationSpace=BoneRelative
    alphaBoolBlend {blendOption=Linear}
    boundBone {boneName=DYN_MSTR_Cube}
    componentLinearAccScale=<0.3 0.3 0.3>
    componentLinearVelScale=<0.6 0.6 0.6>
    physicsBodyDefinitions [{
      sphereCollisionRadius=0
      boundBone {boneName=DYN_MSTR_Cube}
      boxExtents=<10 10 10>
      constraintSetup {angularLimitsMax=<0 0 0>, angularLimitsMin=<0 0 0>, angularTarget=<0 0 0>, linearAxesMax=<0 1 0>, linearAxesMin=<0 -5 0>}
      localJointOffset=<-5 0 0>
     }]
    relativeSpaceBone {boneName=root}
   }
   n9 = "Transform (Modify) Bone - Bone: DYN_Cube"(
    ComponentPose=Local To Component(LocalPose=Local Space Ref Pose())
    Alpha=1
    Translation="0,0,0"
    Rotation=Get Spinning_Rotator_SPD2()
    Scale="1,1,1"
   ) {rotationMode=BMM_Additive, rotationSpace=BCS_BoneSpace, alphaBoolBlend {blendOption=Linear}, boneToModify {boneName=DYN_Cube}}
   n10 = Anim Dynamics - Bone: DYN_MSTR_Cube(ComponentPose=@n9, Alpha=0.7, ExternalForce="0,0,0") {
    angularDampingOverride=0.9
    angularSpringConstant=150
    bGravityOverrideInSimSpace=true
    bLinearSpring=true
    bOverrideAngularDamping=true
    bOverrideLinearDamping=true
    gravityScale=0
    linearDampingOverride=0.999
    linearSpringConstant=50
    simulationSpace=BoneRelative
    alphaBoolBlend {blendOption=Linear}
    boundBone {boneName=DYN_MSTR_Cube}
    componentLinearAccScale=<0.1 0.1 0.1>
    componentLinearVelScale=<0.7 0.7 0.7>
    physicsBodyDefinitions [{
      sphereCollisionRadius=0
      boundBone {boneName=DYN_MSTR_Cube}
      boxExtents=<10 10 10>
      constraintSetup {angularLimitsMax=<0 0 0>, angularLimitsMin=<0 0 0>, angularTarget=<0 0 0>, linearAxesMax=<0 1 0>, linearAxesMin=<0 -3.5 0>}
      localJointOffset=<-5 0 0>
     }]
    relativeSpaceBone {boneName=root}
   }
   n11 = Anim Dynamics - Bone: DYN_MSTR_Cube(ComponentPose=@n10, Alpha=0.55, ExternalForce="0,0,0") {
    angularDampingOverride=0.9
    angularSpringConstant=150
    bGravityOverrideInSimSpace=true
    bLinearSpring=true
    bOverrideAngularDamping=true
    bOverrideLinearDamping=true
    gravityScale=0
    linearDampingOverride=0.999
    linearSpringConstant=250
    simulationSpace=BoneRelative
    alphaBoolBlend {blendOption=Linear}
    boundBone {boneName=DYN_MSTR_Cube}
    componentAppliedLinearAccClamp=<6000 6000 6000>
    componentLinearVelScale=<0.9 0.9 0.9>
    physicsBodyDefinitions [{
      sphereCollisionRadius=0
      boundBone {boneName=DYN_MSTR_Cube}
      boxExtents=<50 50 50>
      constraintSetup {angularLimitsMax=<0 0 0>, angularLimitsMin=<0 0 0>, angularTarget=<0 0 0>, linearAxesMax=<3 0 3>, linearAxesMin=<-3 0 -3>}
      localJointOffset=<-3 0 0>
     }]
    relativeSpaceBone {boneName=root}
   }
   n12 = Get bIsJumping()
   n13 = Get bIsFalling()
   Output Pose(Result=Blend Poses by bool(
     BlendPose_0=Component To Local(ComponentPose=@n11)
     BlendPose_1=Component To Local(ComponentPose=@n8)
     BlendTime_0=0.7
     BlendTime_1=1.2
     bActiveValue=OR Boolean!(A=@n12, B=@n13) -> (ReturnValue)
    ) {blendType=Linear}) {layerGroup=None}
  ]}
 EventGraph {nodes [
   Event BlueprintUpdateAnimation()
   > Sequence()
   n1 = float * float(A=float / float(A=GetRealTimeSeconds(), B=-720) {is=title K2Node_CallFunction}, B=9000)
   > Set Spinning_Rotator(Spinning_Rotator:Rotator=MakeRotator(Roll=@n1, Yaw=@n1)) -> (Output_Get:Rotator)
   n2 = float * float(A=float / float(A=GetRealTimeSeconds(), B=-720), B=50000)
   > Set Spinning_Rotator_SPD2(Spinning_Rotator_SPD2:Rotator=MakeRotator(Roll=@n2, Yaw=@n2)) -> (Output_Get:Rotator)
  ]}
]
variables [
 Spinning_Rotator {type=Rotator}
 Spinning_Rotator_SPD2 {type=Rotator}
]
